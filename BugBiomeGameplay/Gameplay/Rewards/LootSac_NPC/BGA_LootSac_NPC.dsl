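; Hive Stash loot sac that displays a per-instance trapped NPC and frees it when searched or destroyed.
; On ready playlist, the server reserves Living World density, releasing it on opening to request the NPC through an assigned patrol point provider.
; The spawned NPC receives a rescue ability and event, then the sac expires.
; RepNotify bOpened drives mesh removal, burst effects and icon hiding; quest messages credit the rescuer.
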(event Custom|ServerSetOpenedState
  ;[0 0 0] Opened Rep Notify Events
  ;;[0.06301 0.06301 0.06301] Called when bOpened is set on the server only
  (Networking|SwitchHasAuthority
    (:Authority
      (Components|DestroyComponent (Variables|Default|GetLootSacGlassMesh))
      (Collision|SetCollisionEnabled (Variables|Default|GetLootSacGlassMesh) "NoCollision"))))

(event Custom|ClientSetOpenedState
  (bind _lootsacglassmesh (Variables|Default|GetLootSacGlassMesh))
  ;[0 0 0] Opened Rep Notify Events
  ;;[0.06301 0.06301 0.06301] Called when bOpened is set on the client - ensures that if the loot sac falls out of relevancy after it's opened, it won't default to an unopened state
  (Networking|CanExecuteCosmeticEvents
    (:True
      (Niagara|SpawnSystemAttached (Variables|FXandAudio|GetNS_LootSac_Explode) (Variables|Default|GetIdle_Goo_Fx))
      (Rendering|SetVisibility _lootsacglassmesh false false)
      (Components|DestroyComponent _lootsacglassmesh)
      (Audio|Components|Audio|Stop (Variables|Default|GetAudio))
      (Collision|SetCollisionEnabled (Variables|Default|GetLootSacGlassMesh) "NoCollision")
      (MiniMapData|SetVisibility (Variables|Default|GetFortQuestIcon) false)
      (Audio|PlaySoundatLocation (Variables|FXAndAudio|GetOpen%20Sound) (+ (Transformation|GetActorLocation) "0.000000,0.000000,100.000000")))))

(event EventBeginPlay
  ;[0 0 0] BeginPlay - reserve LWM density for the NPC to spawn later
  (CallFunction|SetTrappedNPC)
  (Networking|SwitchHasAuthority
    (:Authority
      (Class|FortAsyncActionCurrentPlaylistReady|PlaylistDataReadyAsync
        (:OnReady
          (Utilities|FlowControl|DelayUntilNextTick)
          (Class|FortAthenaLivingWorldManager|RegisterDensityReservoir (Class|FortAthenaLivingWorldManager|GetLivingWorldManager) self (ScalableFloat|GetValueatLevel (Variables|LivingWorld|GetDensityForNPC)) (ScalableFloat|GetValueatLevel (Variables|LivingWorld|GetDensityRangeForNPC))))))
    (:Remote
      (MiniMapData|SetVisibility (Variables|Default|GetFortQuestIcon) true))))

(event Custom|SpawnNPC
  (bind _returnvalue (Class|FortAthenaLivingWorldManager|GetLivingWorldManager))
  (bind _patrolpathpointprovider (Variables|LivingWorld|GetPatrolPathPointProvider))
  ;[0 0 0] Spawn the NPC at the instance-assigned point provider if it exists
  (Variables|Default|SetbOpened true)
  (Game|SpawnActorfromClass "/BugBiomeGameplay/Gameplay/Rewards/B_LootSac_Destroyed.B_LootSac_Destroyed_C" (Transformation|GetActorTransform) "Undefined" "MultiplyWithRoot" 0 0 0 (Variables|Actor|GetInstigator))
  (Utilities|IsValid _patrolpathpointprovider
    (:Is%20Valid
      (Class|FortAthenaLivingWorldPointProviderInterface|EnablePointProvider_DEPRECATED _patrolpathpointprovider)
      (Class|FortAthenaLivingWorldManager|UnregisterDensityReservoir _returnvalue self)
      (EventDispatchers|BindEventtoOnActorSpawnedDelegate _returnvalue (EventDispatchers|CreateEvent :!event Custom|NPC%20Spawned))
      (Class|FortAthenaLivingWorldManager|RequestEventGeneration _returnvalue)
      (NPC|CallRescuedNPC))))

(event Interaction|EventBlueprintOnInteract
  (params (InteractingPawn :type Fort%20Pawn%20Object%20Reference) (InteractionBeingAttempted :type EInteractionBeingAttempted%20Enum :default "FirstInteraction"))
  ;[0 0 0] When the sack is interacted with, spawn the NPC
  (|CallParentFunction InteractingPawn InteractionBeingAttempted)
  (CallFunction|BroadcastQuestEvent InteractingPawn)
  (Ability|GameplayEffect|AddLooseGameplayTags self (Variables|Default|GetContainerSearchedTags) true)
  (CallFunction|SpawnNPC))

(event Custom|NPC%20Spawned
  (params (SpawnedActor :type Actor%20Object%20Reference))
  ;[0 0 0] Play animation when NPC is spawned
  (Networking|SwitchHasAuthority
    (:Authority
      (Ability|Advanced|GrantandActivateAbilityonSelf (Ability|GetFortAbilitySystemComponent SpawnedActor) (Variables|Default|GetRescueStateAbility))
      (EventDispatchers|UnbindEventfromOnActorSpawnedDelegate (Class|FortAthenaLivingWorldManager|GetLivingWorldManager) (delegate Custom|NPC%20Spawned))
      (NPC|CallOnNPCSpawned SpawnedActor)
      ;; NPC rescues - broadcast a gameplay event so it can be listened to for the NPC
      (Ability|SendGameplayEventtoActor SpawnedActor "(TagName=\"Event.LootSac.NPCRescued\")")
      (Actor|SetLifeSpan 5.0) ; Make sure this actor properly cleans up after it's done spawning the NPC
    )))

(event EventOnDeathServer
  (params (Damage :type Float%20%28single-precision%29) (DamageTags :type Gameplay%20Tag%20Container%20Structure%20%28by%20ref%29) (Momentum :type Vector :default "0, 0, 0") (HitInfo :type Hit%20Result%20Structure%20%28by%20ref%29) (InstigatedBy :type Controller%20Object%20Reference) (DamageCauser :type Actor%20Object%20Reference) (EffectContext :type Gameplay%20Effect%20Context%20Handle%20Structure))
  ;[0 0 0] When this actor is destroyed, if it hasn't already been opened, spawn the NPC
  (if (not (Variables|Default|GetbOpened))
    (CallFunction|BroadcastQuestEvent InstigatedBy)
    (CallFunction|SpawnNPC)
    (else
      (|CallParentFunction Damage DamageTags Momentum HitInfo InstigatedBy DamageCauser EffectContext))))

(event EventOnDamagePlayEffects
  (params (Damage :type Float%20%28single-precision%29) (DamageTags :type Gameplay%20Tag%20Container%20Structure%20%28by%20ref%29) (Momentum :type Vector :default "0, 0, 0") (HitInfo :type Hit%20Result%20Structure%20%28by%20ref%29) (InstigatedBy :type Fort%20Pawn%20Object%20Reference) (DamageCauser :type Actor%20Object%20Reference) (EffectContext :type Gameplay%20Effect%20Context%20Handle%20Structure))
  (bind (_bblockinghit _binitialoverlap _time _distance _location) (Collision|BreakHitResult HitInfo))
  ; On damage, play hitmarker sound for instigator, impact sound for 3p
  (Audio|PlaySoundatLocation "/BugBiomeGameplay/Gameplay/Rewards/Audio/MSS_LootSac_Impact.MSS_LootSac_Impact" _location)
  (if (Pawn|IsLocallyViewed InstigatedBy)
    (Audio|PlaySound2D "/BugBiomeGameplay/Gameplay/Rewards/Audio/MSS_LootSac_HitNotify.MSS_LootSac_HitNotify")))

(fn BlueprintCanInteract
  (params (InteractingPawn :type Fort%20Pawn%20Object%20Reference) (InteractionBeingAttempted :type EInteractionBeingAttempted%20Enum :default "FirstInteraction") (InteractionType :type TInteractionType%20Enum :default "IT_NoInteraction"))
  (return (not (Variables|Default|GetbOpened))))

(fn BlueprintGetInteractionString
  (params (InteractingPawn :type Fort%20Pawn%20Object%20Reference) (InteractionBeingAttempted :type EInteractionBeingAttempted%20Enum :default "FirstInteraction"))
  (return (Variables|NPC|GetInteractText)))

(fn BroadcastQuestEvent
  (params (Instigator :type Actor%20Object%20Reference))
  (Utilities|IsValid Instigator
    (:Is%20Valid
      (CustomVerb|BroadcastCustomVerbMessage (CustomVerb|MakeCustomVerbMessage Instigator "(TagName=\"CustomVerb.LootSac.NPCRescued\")")))))

(fn GetMarkerDisplayData
  (return (Utilities|Struct|MakeMarkedActorDisplayInfo "NSLOCTEXT(\"[3DD9B93CAA20AB4EE16F95C35D5D933B]\", \"D56145D0486C07E7EEA5FFA60432E749\", \"Hive Stash\")" (Variables|Marking|GetMarkerDisplayIcon))))

(fn OnRep_bOpened
  (Networking|SwitchHasAuthority
    (:Authority
      (CallFunction|ServerSetOpenedState))
    (:Remote
      (CallFunction|ClientSetOpenedState))))

(fn PlaySpawnFX)

(fn SetTrappedNPC
  (bind _charactertodisplay (Variables|NPC|GetCharacterToDisplay))
  ;[0 0 0] Trapped NPC is set per-instance from the variable "Character to Display"
  (bind _asbp_loot_sac_npc_mannequin (Utilities|Casting|CastToBP_LootSac_NPC_Mannequin (Variables|/Script/Engine.ChildActorComponent|GetChildActor (Variables|Default|GetMannequin))))
  (Utilities|IsValid _charactertodisplay
    (:Is%20Valid
      (Customization|SetAthenaCharacterDefandVariants _asbp_loot_sac_npc_mannequin _charactertodisplay (Variables|NPC|GetIn%20Override%20Variants)))))

(fn ConstructionScript)
