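; Gameplay ability that plays the break-out animation for an NPC rescued from a loot sac.
; On activation it applies a status-goo effect to itself, pushes AI behavior control so the NPC won't path off to fight, disables collision and switches movement to flying.
; When the montage completes the ability ends, popping behavior control and restoring collision and default land movement.
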(event Ability|EventActivateAbility
  ;[0 0 0] Activate Ability
  (|CallParentFunction)
  (Ability|ApplyGameplayEffectToOwner "/BugBiomeGameplay/Gameplay/Rewards/LootSac_NPC/GE_LootSac_NPC_StatusGoo.GE_LootSac_NPC_StatusGoo_C")
  ;; Ensure the NPC won't path away for engage combat
  (AI|PushBehaviorControl (Variables|InternalData|GetAIController) -2093056)
  (bind _asfort_player_pawn (Utilities|Casting|CastToFortPlayerPawn (Pawn|GetControlledPawn (Variables|InternalData|GetAIController))))
  (bind _output_get (Variables|Default|SetFortPlayerPawn _asfort_player_pawn))
  ; Play the "step down" animation and end the ability
  (Collision|SetActorEnableCollision _output_get false)
  (Pawn|Components|CharacterMovement|SetMovementMode (Variables|/Script/Engine.Character|GetCharacterMovement _output_get) "MOVE_Flying" 0)
  (Ability|Tasks|PlayMontageAndWait "None" (Variables|Default|GetBreakOutAnimation)
    (:OnCompleted
      (Ability|EndAbility))))

(event Ability|EventOnEndAbility
  (params (bWasCancelled :type Boolean))
  ;[0 0 0] End ability
  (|CallParentFunction bWasCancelled)
  ;; Return NPC to normal behavior
  (AI|PopBehaviorControl (Variables|InternalData|GetAIController))
  (Collision|SetActorEnableCollision (Pawn|GetControlledPawn (Variables|InternalData|GetAIController)) true)
  (Pawn|Components|CharacterMovement|SetMovementMode (Variables|/Script/Engine.Character|GetCharacterMovement (Variables|Default|GetFortPlayerPawn)) (Variables|/Script/Engine.CharacterMovementComponent|GetDefaultLandMovementMode (Variables|/Script/Engine.Character|GetCharacterMovement (Variables|Default|GetFortPlayerPawn))) 0))
